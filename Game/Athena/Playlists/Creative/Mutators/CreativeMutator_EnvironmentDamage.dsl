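; Creative-mode environment damage mutator that maps a pickaxe destruction preset onto the legacy DamageMultiplier setting.
; When mutator list properties are applied, it calls the parent handler, then checks the preset.
; For any preset other than the first enum entry, a Select on the enum writes the matching multiplier, which can be zero, one or negative one.
; BeginPlay, Tick and the construction script are empty.
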(event EventBeginPlay)

(event EventTick
  (params (DeltaSeconds :type Float%20%28single-precision%29)))

(event Mutator|EventOnMutatorListPropertiesApplied
  (params (MutatorList :type Fort%20Mutator%20List%20Component%20Object%20Reference))
  (bind _pickaxedestructionpreset (Variables|Default|GetPickaxeDestructionPreset))
  ;[1 0 0.5103] Remap to legacy DamageMultiplier
  (|CallParentFunction MutatorList)
  (if (Utilities|Enum|NotEqual(Enum) _pickaxedestructionpreset "NewEnumerator0")
    (Variables|Mutator|SetDamageMultiplier (Utilities|Select :Index _pickaxedestructionpreset :NewEnumerator0 0.0 :NewEnumerator1 0.0 :NewEnumerator2 1.0 :NewEnumerator3 -1.0))))

(fn ConstructionScript)
