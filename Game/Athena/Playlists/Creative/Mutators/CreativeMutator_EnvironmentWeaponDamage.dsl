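; Creative mutator that sets how much environment damage weapons deal, judging by its name and weapon destruction properties.
; When mutator list properties are applied, it calls the parent, then remaps the destruction preset to the legacy DamageMultiplier.
; The first preset leaves the multiplier untouched; the others select either the destruction percentage as a fraction or a negative sentinel-like value.
; BeginPlay and Tick hold only disabled parent-call stubs.
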(event EventBeginPlay
  (|CallParentFunction) ; This node is disabled and will not be called.\nDrag off pins to build functionality.
)

(event EventTick
  (params (DeltaSeconds :type Float%20%28single-precision%29))
  (|CallParentFunction DeltaSeconds) ; This node is disabled and will not be called.\nDrag off pins to build functionality.
)

(event Mutator|EventOnMutatorListPropertiesApplied
  (params (MutatorList :type Fort%20Mutator%20List%20Component%20Object%20Reference))
  (bind _weapondestructionpreset (Variables|Default|GetWeaponDestructionPreset))
  ;[1 0 0.5103] Remap to legacy DamageMultiplier
  (|CallParentFunction MutatorList)
  (if (Utilities|Enum|NotEqual(Enum) _weapondestructionpreset "NewEnumerator0")
    (Variables|Mutator|SetDamageMultiplier (Utilities|Select :Index _weapondestructionpreset :NewEnumerator0 0.0 :NewEnumerator1 (/ (Variables|Default|GetWeaponDestructionPercentage) 1e+02) :NewEnumerator2 -1.0))))

(fn ConstructionScript
  (|CallParentFunction))
