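; Creative-mode mutator that controls HUD element visibility by applying prioritized visibility modifier data.
; SetHUDVisibilityModifierSourceEnabled, on authority only, multicasts an add or remove of the given visibility data at a priority that defaults to Lowest.
; UpdateHUDVisibilityModifierSource likewise runs only on authority and multicasts updated visibility data.
; The tick event and construction script are empty.
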(event EventTick
  (params (DeltaSeconds :type Float%20%28single-precision%29)))

(fn SetHUDVisibilityModifierSourceEnabled
  (params (bIsEnabled :type Boolean) (HUDElementVisibilityData :type HUDElement%20Visibility%20Rep%20Data%20Structure) (Priority :type EFortHUDElementVisibiltyPriority%20Enum :default "Lowest"))
  (Networking|SwitchHasAuthority
    (:Authority
      (if bIsEnabled
        (CallFunction|MulticastAddVisibilityModData Priority HUDElementVisibilityData)
        (else
          (CallFunction|MulticastRemoveVisibilityModData Priority HUDElementVisibilityData))))))

(fn UpdateHUDVisibilityModifierSource
  (params (HUDElementVisibilityData :type HUDElement%20Visibility%20Rep%20Data%20Structure))
  (Networking|SwitchHasAuthority
    (:Authority
      (CallFunction|MulticastUpdateVisibilityModData HUDElementVisibilityData))))

(fn ConstructionScript)
