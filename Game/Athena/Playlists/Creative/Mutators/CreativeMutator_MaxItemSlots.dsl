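; Creative mutator that sets the maximum item slot count from the maximum equipment slots property.
; When mutator list properties are applied, it calls the parent handler and checks the override flag.
; With the override on, a non-negative value is converted to float and written to MaxItemSlots, while a negative value leaves it untouched.
; With the override off, MaxItemSlots is set to -1.
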(event EventBeginPlay)

(event EventTick
  (params (DeltaSeconds :type Float%20%28single-precision%29)))

(event Mutator|EventOnMutatorListPropertiesApplied
  (params (MutatorList :type Fort%20Mutator%20List%20Component%20Object%20Reference))
  (bind _maximumequipmentslots (Variables|Default|GetMaximumEquipmentSlots))
  (|CallParentFunction MutatorList)
  (if (Variables|Default|GetMaximumEquipmentSlots_Override)
    (if (>= _maximumequipmentslots 0)
      (Variables|AllowItemDrop|SetMaxItemSlots (Math|Conversions|ToFloat(Integer) _maximumequipmentslots)))
    (else
      (Variables|AllowItemDrop|SetMaxItemSlots -1.0))))

(fn ConstructionScript)
