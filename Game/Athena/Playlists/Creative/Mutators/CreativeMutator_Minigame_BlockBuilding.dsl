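; Creative minigame mutator for the Allow Building setting, active from minigame start until minigame end.
; During play it supplies a no-building HUD visibility tag when building mode is None and minigame settings do not already hide the build resource HUD.
; When mutator list properties apply, a DoOnce-guarded migration converts the deprecated Block Building knob to TrapsOnly if Allow Building is still All, rewriting the property overrides.
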(event MinigameStartedDelegate(FortMinigameLogic)
  (Mutator|SetMutatorActive true))

(event MinigameEndedDelegate(FortMinigameLogic)
  (Mutator|SetMutatorActive false))

(event Mutator|EventOnMutatorListPropertiesApplied
  (params (MutatorList :type Fort%20Mutator%20List%20Component%20Object%20Reference))
  (Utilities|IsValid MutatorList
    (:Is%20Valid
      ;[1 1 1] Changes to the properties list can trigger this function more then once.
      (Utilities|FlowControl|DoOnce false
        (:Completed
          (Utilities|FlowControl|Sequence
            (:then_0
              _)
            (:then_1
              _))
          (CallFunction|MigrateBlockBuildingUserKnob MutatorList) ; 11.20 Migration:\nTrue (Allow Building OFF) -> BuildingMode::TrapsOnly,\nFalse (Allow Building ON) -> BuildingMode::All
        )))))

(fn GetHUDVisibilityTags
  (params (Pawn :type Pawn%20Object%20Reference))
  (bind _returnvalue (Minigame|GetMinigameSettingsMachine (Minigame|GetMinigame Pawn :!class /Script/FortniteGame.MinigameBlueprintLibrary)))
  (bind _returnvalue_2 (Minigame|GetMinigame Pawn :!class /Script/FortniteGame.MinigameBlueprintLibrary))
  (Utilities|IsValid Pawn
    (:Is%20Valid
      (Utilities|IsValid _returnvalue_2
        (:Is%20Valid
          (Minigame|IsInProgress _returnvalue_2
            (:Yes
              (Utilities|IsValid _returnvalue
                (:Is%20Valid
                  (if (not (Class|FortMinigameSettingsBuilding|ShouldHideBuildResourceHUD _returnvalue))
                    (switch Utilities|FlowControl|Switch|SwitchonEBuildingMode (Variables|BlockBuilding|GetAllowBuilding)
                      (:None
                        (return (Variables|Default|GetNoBuildingHUDVisibilityTag)))
                      (:BuildingsOnly
                        _)
                      (:TrapsOnly
                        _)
                      (:All
                        _))
                    (return 0)))
                (:Is%20Not%20Valid
                  _)))
            (:No
              _)))
        (:Is%20Not%20Valid
          _)))
    (:Is%20Not%20Valid
      _)))

(fn MigrateBlockBuildingUserKnob
  (params (Mutator%20List :type Fort%20Mutator%20List%20Component%20Object%20Reference))
  ;[1 1 1] If the deprecated 'Block Building' setting is true, and 'Allow Building' is the default of all
  (if (and (Variables|MigratedKnobs|GetbBlockBuilding) (Utilities|Enum|Equal(Enum) (Variables|BlockBuilding|GetAllowBuilding) "All"))
    ;[1 1 1] Add the new property for 'Allow Building' and set it to traps only
    (Variables|BlockBuilding|SetAllowBuilding "TrapsOnly")
    (Modifiers|AddorUpdatePropertyOverrides Mutator%20List (Utilities|Map|MakeMap (Class|PlaylistUserOptionBase|GetOptionKey (Variables|Default|GetAllow%20Building%20Definition) true 0) (Class|PlaylistUserOptionBase|GetOptionValueNamefromValue (Variables|Default|GetAllow%20Building%20Definition) "TrapsOnly" true 0)))
    ;[1 1 1] Remove the old block building property and set the boolean to false
    (Modifiers|RemovePropertyOverride Mutator%20List (Class|PlaylistUserOptionBase|GetOptionKey (Variables|Default|GetDeprecated%20Block%20Building%20Definition) true 0))
    (Variables|MigratedKnobs|SetbBlockBuilding false)))

(fn ConstructionScript)
