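; Creative minigame mutator that controls environment damage during Creative games in Fortnite.
; It activates when the minigame starts and deactivates on post-game end, abandon or reset states, with a last-round check so it stays active between rounds.
; When mutator list properties apply, it calls the parent, then remaps any non-default EnvironmentDamagePreset to an Off, All or PlayerBuiltOnly damage filter that sets the legacy DamageMultiplier.
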(event MinigameStartedDelegate(FortMinigameLogic)
  (Mutator|SetMutatorActive true))

(event MinigameStateChangedDelegate(FortMinigameLogic)
  (params (Minigame :type Fort%20Minigame%20Object%20Reference) (NewMinigameState :type EFortMinigameState%20Enum :default "PreGame"))
  (switch Utilities|FlowControl|Switch|SwitchonEFortMinigameState NewMinigameState
    (:PostGameEnd
      ;[1 1 1] Only disable the mutator if we are actually ending the game.  If we are just going between rounds leave the mutator active
      (bind _currentminigame (Variables|/Script/FortniteGame.FortMinigameLogicComponent|GetCurrentMinigame (Variables|Default|GetFortMinigameLogic))
        (:then
          (if (Minigame|IsLastRound _currentminigame)
            _))
        (:else
          _))
      (Mutator|SetMutatorActive false))
    (:PostGameAbandon
      _)
    (:PostGameReset
      _))
  ;[1 1 1] Always disble mutator when abandoning the game so it is not left on.
  (Mutator|SetMutatorActive false))

(event Mutator|EventOnMutatorListPropertiesApplied
  (params (MutatorList :type Fort%20Mutator%20List%20Component%20Object%20Reference))
  (bind _environmentdamagepreset (Variables|Default|GetEnvironmentDamagePreset))
  ;[1 0 0.5103] Remap to legacy DamageMultiplier
  (|CallParentFunction MutatorList)
  (if (Utilities|Enum|NotEqual(Enum) _environmentdamagepreset "NewEnumerator0")
    (Variables|Mutator|SetDamageMultiplier (CallFunction|GetValuefromDamageFilter (Utilities|Select :Index _environmentdamagepreset :NewEnumerator0 "Off" :NewEnumerator2 "Off" :NewEnumerator1 "All" :NewEnumerator3 "PlayerBuiltOnly")))))

(fn ConstructionScript)
