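; Creative mutator that sets the player pickaxe damage multiplier in Fortnite Creative from a preset option.
; When mutator list properties are applied, it calls the parent handler, then maps the PickaxeDamagePreset enum through a Select.
; The result is written to the legacy DamageMultiplier variable, so the preset choice drives that older property.
; BeginPlay, Tick and the construction script contain no logic.
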(event EventBeginPlay)

(event EventTick
  (params (DeltaSeconds :type Float%20%28single-precision%29)))

(event Mutator|EventOnMutatorListPropertiesApplied
  (params (MutatorList :type Fort%20Mutator%20List%20Component%20Object%20Reference))
  ;[1 0 0.5103] Remap to legacy DamageMultiplier
  (|CallParentFunction MutatorList)
  (Variables|Mutator|SetDamageMultiplier (Utilities|Select :Index (Variables|Default|GetPickaxeDamagePreset) :NewEnumerator0 1.0 :NewEnumerator2 0.0)))

(fn ConstructionScript)
