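; Creative-mode mutator Blueprint, named for respawn location, that flags resurrected player pawns to play respawn effects.
; On the player pawn resurrected event, it casts the pawn to PlayerPawn_Athena and sets PlayRespawnFXOnSpawn to true.
; BeginPlay, Tick and the construction script are empty, and no location logic appears in this graph.
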(event EventBeginPlay)

(event EventTick
  (params (DeltaSeconds :type Float%20%28single-precision%29)))

(event EventOnPlayerPawnResurrected
  (params (PlayerPawn :type Fort%20Player%20Pawn%20Object%20Reference))
  (bind _asplayer_pawn_athena (Utilities|Casting|CastToPlayerPawn_Athena PlayerPawn))
  (Variables|/Game/Athena/PlayerPawn_Athena.PlayerPawn_Athena_C|SetPlayRespawnFXOnSpawn true _asplayer_pawn_athena))

(fn ConstructionScript)
