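; Creative mutator that applies team settings to a minigame's game rules for each team context.
; SetupGameRulesForTeamContext looks up each enabled setting's component to set initial team size, size weight and team color.
; Kill, AI kill, item collection, score and object destruction targets become stat queries in include or exclude lists via OverrideEndCondition.
; CommitEndConditionOverrides then adds or removes those end conditions, globally for team index 255, otherwise per team.
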(fn CommitEndConditionOverrides
  (params (Minigame :type Fort%20Minigame%20Object%20Reference) (IncludeList :type Array%20of%20Fort%20Minigame%20Stat%20Query%20Structures) (ExcludeList :type Array%20of%20Fort%20Minigame%20Stat%20Query%20Structures) (TeamIndex :type Byte))
  (bind _minigame (Variables|GetMinigame))
  (bind _endconditioncomponent (Variables|/Script/FortniteGame.FortMinigame|GetEndConditionComponent _minigame))
  (bind _minigame_2 (Variables|GetMinigame))
  (bind _endconditioncomponent_2 (Variables|/Script/FortniteGame.FortMinigame|GetEndConditionComponent _minigame_2))
  (LocalVariables|SetbDefaultTeam (Math|Byte|Equal(Byte) TeamIndex 255))
  (Utilities|FlowControl|Sequence
    (:then_0
      (for _array_element IncludeList
        (if (LocalVariables|GetbDefaultTeam)
          (Minigame|GameRules|AddorOverrideEndCondition _endconditioncomponent _array_element)
          (else
            (Minigame|GameRules|AddorOverrideEndConditionforTeam _endconditioncomponent (Variables|GetTeamIndex) _array_element)))))
    (:then_1
      (for _array_element_2 ExcludeList
        (if (LocalVariables|GetbDefaultTeam)
          (Minigame|GameRules|RemoveEndCondition _endconditioncomponent_2 _array_element_2)
          (else
            (Minigame|GameRules|RemoveEndConditionforTeam _endconditioncomponent_2 (Variables|GetTeamIndex) _array_element_2)))))))

(fn OverrideEndCondition
  (params (Stat :type Fort%20Minigame%20Stat%20Filter%20Class%20Reference) (Scope :type EFortMinigameStatScope%20Enum :default "Team") (Value :type Integer) (bStaticCount :type Boolean) (bInclude :type Boolean) (IncludeList :type Array%20of%20Fort%20Minigame%20Stat%20Query%20Structures) (ExcludeList :type Array%20of%20Fort%20Minigame%20Stat%20Query%20Structures))
  (bind _self self)
  (bind _returnvalue (Minigame|Stats|GetMinigameStatsComponent _self))
  (bind _returnvalue_2 (Class|FortMinigameStatsComponent|GetLoadedStatfromClass _returnvalue Stat))
  (bind _fortminigamestatquery (Utilities|Struct|MakeFortMinigameStatQuery _returnvalue_2 Scope true "GreaterOrEqual" Value bStaticCount))
  (Utilities|IsValid _returnvalue_2
    (:Is%20Valid
      (if bInclude
        (Utilities|Array|Add (Variables|GetIncludeList) _fortminigamestatquery)
        (else
          (Utilities|Array|Add (Variables|GetExcludeList) _fortminigamestatquery))))))

(fn SetupGameRulesForTeamContext
  (params (Minigame :type Fort%20Minigame%20Object%20Reference))
  (bind _returnvalue (Minigame|GetMinigameTeamComponent Minigame))
  (bind _returnvalue_2 (TeamSettings|GetTeamContext))
  (bind _returnvalue_3 (|CallParentFunction Minigame))
  (LocalVariables|SetbOverridden _returnvalue_3)
  (Utilities|IsValid Minigame
    (:Is%20Valid
      (Utilities|IsValid _returnvalue
        (:Is%20Valid
          (LocalVariables|SetLocalMiniGameTeamComponent _returnvalue)
          (LocalVariables|SetTeamIndex _returnvalue_2)
          (LocalVariables|SetbDefaultTeam (Math|Byte|Equal(Byte) _returnvalue_2 255))
          (Utilities|FlowControl|Sequence
            (:then_0
              ;[1 1 1] Always fall back to defaults, in case there isn't an override for this specific team
              (bind _output_get (LocalVariables|SetTeamSettings (TeamSettings|GetComponentforEnabledSetting (Variables|Default|GetMaxInitTeamSizeName))))
              (bind _ascreative_team_settings (Utilities|Casting|CastToCreativeTeamSettings _output_get))
              (Minigame|GameRules|SetMaxInitTeamSizeforTeam (LocalVariables|GetLocalMiniGameTeamComponent) (LocalVariables|GetTeamIndex) (select (> (Variables|/Game/Athena/Playlists/Creative/Mutators/CreativeTeamSettings.CreativeTeamSettings_C|GetMaxInitTeamSize _ascreative_team_settings) 0) (Variables|/Game/Athena/Playlists/Creative/Mutators/CreativeTeamSettings.CreativeTeamSettings_C|GetMaxInitTeamSize _ascreative_team_settings) -1))
              (LocalVariables|SetbOverridden true))
            (:then_1
              ;[1 1 1] Always fall back to defaults, in case there isn't an override for this specific team
              (bind _output_get_2 (LocalVariables|SetTeamSettings (TeamSettings|GetComponentforEnabledSetting (Variables|Default|GetInitTeamSizeWeightName))))
              (bind _ascreative_team_settings_2 (Utilities|Casting|CastToCreativeTeamSettings _output_get_2))
              (Minigame|GameRules|SetInitTeamSizeWeightforTeam (LocalVariables|GetLocalMiniGameTeamComponent) (LocalVariables|GetTeamIndex) (Variables|/Game/Athena/Playlists/Creative/Mutators/CreativeTeamSettings.CreativeTeamSettings_C|GetInitTeamSizeWeight _ascreative_team_settings_2))
              (LocalVariables|SetbOverridden true))
            (:then_2
              (bind _output_get_3 (LocalVariables|SetTeamSettings (TeamSettings|GetComponentforEnabledSetting (Variables|Default|GetNumKillsForEndConditionName) (LocalVariables|GetbDefaultTeam))))
              (bind _ascreative_team_settings_3 (Utilities|Casting|CastToCreativeTeamSettings _output_get_3))
              (CallFunction|OverrideEndCondition "/Game/Athena/Playset/Minigames/StatFilters/MinigameStat_PlayerKills.MinigameStat_PlayerKills_C" "Team" (Variables|/Game/Athena/Playlists/Creative/Mutators/CreativeTeamSettings.CreativeTeamSettings_C|GetNumKillsForEndCondition _ascreative_team_settings_3) true (> (Variables|/Game/Athena/Playlists/Creative/Mutators/CreativeTeamSettings.CreativeTeamSettings_C|GetNumKillsForEndCondition _ascreative_team_settings_3) 0) (LocalVariables|GetEndConditionsToInclude) (LocalVariables|GetEndConditionsToExclude))
              (LocalVariables|SetbOverridden true))
            (:then_3
              (bind _output_get_4 (LocalVariables|SetTeamSettings (TeamSettings|GetComponentforEnabledSetting (Variables|Default|GetNumCollectItemsForEndConditionName) (LocalVariables|GetbDefaultTeam))))
              (bind _ascreative_team_settings_4 (Utilities|Casting|CastToCreativeTeamSettings _output_get_4))
              (CallFunction|OverrideEndCondition "/Game/Athena/Playset/Minigames/StatFilters/MinigameStat_Challenge_All.MinigameStat_Challenge_All_C" "Team" (Variables|/Game/Athena/Playlists/Creative/Mutators/CreativeTeamSettings.CreativeTeamSettings_C|GetNumCollectItemsForEndCondition _ascreative_team_settings_4) (> (Variables|/Game/Athena/Playlists/Creative/Mutators/CreativeTeamSettings.CreativeTeamSettings_C|GetNumCollectItemsForEndCondition _ascreative_team_settings_4) 0) (!= (Variables|/Game/Athena/Playlists/Creative/Mutators/CreativeTeamSettings.CreativeTeamSettings_C|GetNumCollectItemsForEndCondition _ascreative_team_settings_4) 0) (LocalVariables|GetEndConditionsToInclude) (LocalVariables|GetEndConditionsToExclude))
              (LocalVariables|SetbOverridden true))
            (:then_4
              (bind _output_get_5 (LocalVariables|SetTeamSettings (TeamSettings|GetComponentforEnabledSetting (Variables|Default|GetCollectItemsToEndName) (LocalVariables|GetbDefaultTeam))))
              (bind _ascreative_team_settings_5 (Utilities|Casting|CastToCreativeTeamSettings _output_get_5))
              (CallFunction|OverrideEndCondition "/Game/Athena/Playset/Minigames/StatFilters/MinigameStat_Challenge_All.MinigameStat_Challenge_All_C" "Team" (Utilities|Select :Index (Variables|/Game/Athena/Playlists/Creative/Mutators/CreativeTeamSettings.CreativeTeamSettings_C|GetCollectItemsToEnd _ascreative_team_settings_5) :NewEnumerator2 0 :NewEnumerator0 -1 :NewEnumerator1 (Variables|/Game/Athena/Playlists/Creative/Mutators/CreativeTeamSettings.CreativeTeamSettings_C|GetCollectItemCountToEnd _ascreative_team_settings_5)) (> (Utilities|Select :Index (Variables|/Game/Athena/Playlists/Creative/Mutators/CreativeTeamSettings.CreativeTeamSettings_C|GetCollectItemsToEnd _ascreative_team_settings_5) :NewEnumerator2 0 :NewEnumerator0 -1 :NewEnumerator1 (Variables|/Game/Athena/Playlists/Creative/Mutators/CreativeTeamSettings.CreativeTeamSettings_C|GetCollectItemCountToEnd _ascreative_team_settings_5)) 0) (!= (Utilities|Select :Index (Variables|/Game/Athena/Playlists/Creative/Mutators/CreativeTeamSettings.CreativeTeamSettings_C|GetCollectItemsToEnd _ascreative_team_settings_5) :NewEnumerator2 0 :NewEnumerator0 -1 :NewEnumerator1 (Variables|/Game/Athena/Playlists/Creative/Mutators/CreativeTeamSettings.CreativeTeamSettings_C|GetCollectItemCountToEnd _ascreative_team_settings_5)) 0) (LocalVariables|GetEndConditionsToInclude) (LocalVariables|GetEndConditionsToExclude))
              (LocalVariables|SetbOverridden true))
            (:then_5
              (bind _output_get_6 (LocalVariables|SetTeamSettings (TeamSettings|GetComponentforEnabledSetting (Variables|Default|GetNumAIKillsForEndConditionName) (LocalVariables|GetbDefaultTeam))))
              (bind _ascreative_team_settings_6 (Utilities|Casting|CastToCreativeTeamSettings _output_get_6))
              (CallFunction|OverrideEndCondition "/Game/Athena/Playset/Minigames/StatFilters/MinigameStat_AIKills.MinigameStat_AIKills_C" "Team" (Variables|/Game/Athena/Playlists/Creative/Mutators/CreativeTeamSettings.CreativeTeamSettings_C|GetNumAIKillsForEndCondition _ascreative_team_settings_6) true (> (Variables|/Game/Athena/Playlists/Creative/Mutators/CreativeTeamSettings.CreativeTeamSettings_C|GetNumAIKillsForEndCondition _ascreative_team_settings_6) 0) (LocalVariables|GetEndConditionsToInclude) (LocalVariables|GetEndConditionsToExclude))
              (LocalVariables|SetbOverridden true))
            (:then_6
              (bind _output_get_7 (LocalVariables|SetTeamSettings (TeamSettings|GetComponentforEnabledSetting (Variables|Default|GetScoreForEndConditionName) (LocalVariables|GetbDefaultTeam))))
              (bind _ascreative_team_settings_7 (Utilities|Casting|CastToCreativeTeamSettings _output_get_7))
              (CallFunction|OverrideEndCondition "/Game/Athena/Playset/Minigames/StatFilters/MinigameStat_Score.MinigameStat_Score_C" "Team" (Variables|/Game/Athena/Playlists/Creative/Mutators/CreativeTeamSettings.CreativeTeamSettings_C|GetScoreForEndCondition _ascreative_team_settings_7) true (> (Variables|/Game/Athena/Playlists/Creative/Mutators/CreativeTeamSettings.CreativeTeamSettings_C|GetScoreForEndCondition _ascreative_team_settings_7) 0) (LocalVariables|GetEndConditionsToInclude) (LocalVariables|GetEndConditionsToExclude))
              (LocalVariables|SetbOverridden true))
            (:then_7
              (bind _output_get_8 (LocalVariables|SetTeamSettings (TeamSettings|GetComponentforEnabledSetting (Variables|Default|GetNumObjectDestroysForWinName) (LocalVariables|GetbDefaultTeam))))
              (bind _ascreative_team_settings_8 (Utilities|Casting|CastToCreativeTeamSettings _output_get_8))
              (CallFunction|OverrideEndCondition "/Game/Athena/Playset/Minigames/StatFilters/MinigameStat_Destructible_All.MinigameStat_Destructible_All_C" "Team" (Variables|/Game/Athena/Playlists/Creative/Mutators/CreativeTeamSettings.CreativeTeamSettings_C|GetNumObjectDestroysForWin _ascreative_team_settings_8) (> (Variables|/Game/Athena/Playlists/Creative/Mutators/CreativeTeamSettings.CreativeTeamSettings_C|GetNumObjectDestroysForWin _ascreative_team_settings_8) 0) (!= (Variables|/Game/Athena/Playlists/Creative/Mutators/CreativeTeamSettings.CreativeTeamSettings_C|GetNumObjectDestroysForWin _ascreative_team_settings_8) 0) (LocalVariables|GetEndConditionsToInclude) (LocalVariables|GetEndConditionsToExclude))
              (LocalVariables|SetbOverridden true))
            (:then_8
              (CallFunction|CommitEndConditionOverrides (Variables|GetMinigame) (LocalVariables|GetEndConditionsToInclude) (LocalVariables|GetEndConditionsToExclude) (LocalVariables|GetTeamIndex))
              (bind _output_get_9 (LocalVariables|SetTeamSettings (TeamSettings|GetComponentforEnabledSetting (Variables|Default|GetTeamColorName))))
              (bind _ascreative_team_settings_9 (Utilities|Casting|CastToCreativeTeamSettings _output_get_9))
              (Minigame|GameRules|SetTeamColorIndexforTeam (LocalVariables|GetLocalMiniGameTeamComponent) (LocalVariables|GetTeamIndex) (Math|Conversions|ToInteger(Byte) (Variables|/Game/Athena/Playlists/Creative/Mutators/CreativeTeamSettings.CreativeTeamSettings_C|GetTeamColor _ascreative_team_settings_9))))
            (:then_9
              _)))
        (:Is%20Not%20Valid
          _))
      (return (LocalVariables|GetbOverridden)))
    (:Is%20Not%20Valid
      _)))

(fn ConstructionScript)
