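; CreativeTeamSettings is a Creative mutator for team settings, which supplies HUD visibility tags based on the building mode.
; GetComponentHUDVisibilityTags switches on the mutator's building mode.
; The None mode returns the no-building HUD visibility tag; the buildings-only, traps-only and all modes fall through to an empty default.
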(fn GetComponentHUDVisibilityTags
  (switch Utilities|FlowControl|Switch|SwitchonEBuildingMode (Variables|Mutator|Building|GetBuildingMode)
    (:None
      (return (Variables|HUDVisibilityTags|GetNoBuildingHUDVisibilityTag)))
    (:BuildingsOnly
      _)
    (:TrapsOnly
      _)
    (:All
      _))
  (return 0))
